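# Landmark mountain surface material, Default Lit and usable on instanced static meshes, that layers snow over base color, roughness and specular.
# Snow coverage comes from a mask texture combined with tiled noise and an offset, then sharpened by a power and capped at a maximum opacity.
# The normal map is transformed out of tangent space and dotted with the sky atmosphere light direction, subtly shifting snow color toward a blue shadow tint away from the light.
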Material /Hermes/Terrain/Mountain/M_Landmark_Mountain
graph {
 properties {bAutomaticallySetUsageInEditor=false, bTangentSpaceNormal=false, bUsedWithInstancedStaticMeshes=true, usedShadingModels=Default Lit}
 nodes [
  n1 = NamedRerouteDeclaration(None=Transform(None=TextureSampleParameter2D() {parameterName=Normal, samplerType=SAMPLERTYPE_Normal, texture=&/T_Landmark_Mountain_N}.RGB)) {name=Normals, nodeColor=<rgba 1 0.2344 0 1>}
  n2 = SkyAtmosphereLightDirection()
  n3 = Add(A=Multiply(A=DotProduct(A=@n1, B=@n2)) {constB=-0.1}) {constB=0.3}
  LinearInterpolate(
   A=TextureSampleParameter2D() {parameterName=Base Color, texture=&/T_Landmark_Mountain_D}.RGB
   B=LinearInterpolate(
    A=VectorParameter() {parameterName=Snow Color, defaultValue=<rgba 0.4635 0.4482 0.4021 1>}.RGB
    B=VectorParameter() {parameterName=Snow Shadow Color, defaultValue=<rgba 0 0.0742 0.2656 1>}.RGB
    Alpha=Saturate(None=@n3)
   )
   Alpha=NamedRerouteUsage()
  )
  LinearInterpolate(
   A=ScalarParameter() {defaultValue=0.9, parameterName=Base Roughness}
   B=ScalarParameter() {defaultValue=0.6, parameterName=Snow Roughness}
   Alpha=NamedRerouteUsage()
  )
  LinearInterpolate(
   A=ScalarParameter() {defaultValue=0.5, parameterName=Base Specular}
   B=ScalarParameter() {defaultValue=0.6, parameterName=Snow Specular}
   Alpha=NamedRerouteUsage()
  )
  n4 = TextureSampleParameter2D() {parameterName=Snow Mask, samplerType=SAMPLERTYPE_LinearColor, texture=&/T_Landmark_Mountain_M}
  n5 = TextureCoordinate()
  n6 = ScalarParameter() {defaultValue=0.15, parameterName=Snow Noise Scale}
  n7 = Add(
   A=@n4.R
   B=TextureSample(UVs=Divide(A=@n5, B=@n6)) {samplerType=SAMPLERTYPE_LinearGrayscale, texture=&/T_Landmark_Mountain_Noise}.R
  )
  n8 = ScalarParameter() {defaultValue=0.1242, parameterName=Sno Mask Offset}
  n9 = Power(Base=Multiply(A=@n4.R, B=Add(A=@n7, B=@n8)), Exp=ScalarParameter() {defaultValue=3, parameterName=Snow Mask Exp}) {constExponent=8}
  NamedRerouteDeclaration(None=Min(A=Saturate(None=@n9), B=ScalarParameter() {defaultValue=0.9, parameterName=Snow Max Opacity})) {name=Snow Mask, nodeColor=<rgba 0 1 0.4844 1>}
  NamedRerouteUsage()
 ]
}
